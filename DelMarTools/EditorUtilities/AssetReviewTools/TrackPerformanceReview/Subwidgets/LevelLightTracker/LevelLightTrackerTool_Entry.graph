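# Editor utility entry widget for the Level Light Tracker track-review tool, summarizing one group of light-bearing actors.
# UpdateInfo fills the name, component count and light-setting labels, while AddActor collects actors and updates the instance count.
# Buttons select those actors or their building level instances, shown only when bBuildingInstance is set, then align the viewport camera.
# If the first actor is invalid, selection falls back to all actors of ClassRef.
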EditorUtilityWidgetBlueprint /DelMarTools/EditorUtilities/AssetReviewTools/TrackPerformanceReview/Subwidgets/LevelLightTracker/LevelLightTrackerTool_Entry
graphs [
 AddActor {nodes [
   n1 = AddActor!() -> (then, Actor:Actor Object Reference) {is=title K2Node_FunctionEntry}
   > Add(
    TargetArray:Array of Actor Object References=Get ActorsInLevel() -> (ActorsInLevel:Array of Actor Object References)
    NewItem:"Actor Object Reference (by ref)"=@n1
   ) {is=title K2Node_CallArrayFunction}
   n2 = Get ActorCount() -> (ActorCount:int=0)
   > Increment Int(Value=@n2)
   n3 = "integer > integer"(A=@n2, B=1)
   > "SetText (Text)"(
    self=Get Text_ActorCount() -> (Text_ActorCount:Text Object Reference)
    InText=Format Text!(Format="LOCTEXT(\"{Count} Instance{s}\")", Count:int=@n2, s:String=SelectString(A=s, bPickA=@n3)) -> (Result)
   )
  ]}
 AddBuildingLevelInstance {nodes [
   n1 = AddBuildingLevelInstance!() -> (then, BuildingLevelInstance:Actor Object Reference) {is=title K2Node_FunctionEntry}
   > Add(TargetArray=Get BuildingLevelInstances() -> (BuildingLevelInstances:Array of Actor Object References), NewItem=@n1)
  ]}
 EventGraph {nodes [
   "On Clicked (Button_SelectInstances)"() {is=title K2Node_ComponentBoundEvent}
   > SelectAffectedLevelActors!(execute, self, ActorsToSelect:Array of Actor Object References=Get BuildingLevelInstances()) -> (then)
   "On Clicked (Button_SelectLights)"() {is=title K2Node_ComponentBoundEvent}
   > SelectAffectedLevelActors(ActorsToSelect=Get ActorsInLevel())
   Event Construct()
   > SetVisibility(
    self=[
     Get Text_BuildingLevelInstanceWarning() -> (Text_BuildingLevelInstanceWarning:Text Object Reference)
     Get Button_SelectInstances() -> (Button_SelectInstances:Editor Utility Button Object Reference)
    ]
    InVisibility=Select(Option 0=Collapsed, Option 1=Visible, Index=Get bBuildingInstance() -> (bBuildingInstance:bool))
   )
  ]}
 SelectAffectedLevelActors {nodes [
   n1 = SelectAffectedLevelActors!() -> (then, ActorsToSelect:Array of Actor Object References) {is=title K2Node_FunctionEntry}
   > n2 = Is Valid(InputObject="Get (a copy)"(Array:Array of Actor Object References=@n1) -> (Output:Actor Object Reference))
   n3 = GetAllActorsOfClass!(
    execute=@n2.Is Not Valid
    WorldContextObject:Object Reference=Self-Reference()
    ActorClass:Actor Class Reference=Get ClassRef() -> (ClassRef:Actor Class Reference)
   ) -> (then, OutActors:Array of Actor Object References)
   > n4 = SetSelectedLevelActors(self=Get EditorActorSubsystem(), ActorsToSelect=@n3)
   ExecuteConsoleCommand(execute=[@n4.then, @n5.then], Command=CAMERA ALIGN ACTIVEVIEWPORTONLY)
   n5 = SetSelectedLevelActors(
    execute=@n2.Is Valid
    self=Get EditorActorSubsystem()
    ActorsToSelect=Get ActorsToSelect() -> (ActorsToSelect:Array of Actor Object References)
   )
  ]}
 UpdateInfo {nodes [
   UpdateInfo!() -> (
    then
    ActorClassName:String
    ComponentClassName:String
    LightCountPerActor:int
    bVisible:bool=~
    bAffectsWorld:bool=~
    bCastShadows:bool=~
    Mobility:EComponentMobility Enum=Static
    Detail Mode:EDetailMode Enum=DM_Low
    MaxDrawDistance:double=~
   ) {is=title K2Node_FunctionEntry}
   > n1 = Sequence!(execute) -> (then_0, then_1, then_2:Exec, then_3:Exec, then_4:Exec, then_5:Exec, then_6:Exec, then_7:Exec, then_8:Exec)
   > Set ActorClassName(ActorClassName:String=Get ActorClassName() -> (ActorClassName:String)) -> (Output_Get:String)
   > "SetText (Text)"(self=Get Text_ActorName() -> (Text_ActorName:Text Object Reference), InText="To Text (String)"(InString=Get ActorClassName()))
   Set ComponentClassName(execute=@n1.then_1, ComponentClassName:String=Get ComponentClassName() -> (ComponentClassName:String)) -> (Output_Get:String)
   > Set LightCountPerActor(LightCountPerActor:int=Get LightCountPerActor() -> (LightCountPerActor:int=0)) -> (Output_Get:int=0)
   n2 = Get LightCountPerActor() -> (LightCountPerActor=0)
   n3 = "integer > integer"(A=@n2, B=1)
   > "SetText (Text)"(
    self=Get Text_ComponentName() -> (Text_ComponentName:Text Object Reference)
    InText=Format Text!(
     Format="LOCTEXT(\"{amount} {componentName}{s}\")"
     amount:int=@n2
     componentName:String=Get ComponentClassName()
     s=SelectString(A=s, bPickA=@n3)
    ) -> (Result)
   )
   n4 = Set Mobility(execute=@n1.then_3, Mobility:EComponentMobility Enum=Get Mobility() -> (Mobility:EComponentMobility Enum=Static)) -> (Output_Get:EComponentMobility Enum=Static)
   > "SetText (Text)"(
    self=Get Text_Mobility() -> (Text_Mobility:Text Object Reference)
    InText="To Text (String)"(InString=Enum to String!(Enumerator:EComponentMobility Enum=@n4) -> (ReturnValue:String))
   )
   n5 = Get MaxDrawDistance() -> (MaxDrawDistance:double)
   Set MaxDrawDistance(execute=@n1.then_4, MaxDrawDistance:double=@n5) -> (Output_Get:double)
   > "SetText (Text)"(
    self=Get Text_MaxDrawDistance() -> (Text_MaxDrawDistance:Text Object Reference)
    InText=Format Text!(Format="LOCTEXT(\"MaxDrawDistance: {distance} \")", distance:double=@n5) -> (Result)
   )
   n6 = Get bAffectsWorld() -> (bAffectsWorld:bool)
   Set bAffectsWorld(execute=@n1.then_5, bAffectsWorld:bool=@n6) -> (Output_Get:bool)
   > "SetText (Text)"(
    self=Get Text_AffectsWorld() -> (Text_AffectsWorld:Text Object Reference)
    InText=Format Text!(Format="LOCTEXT(\"AffectsWorld: {affectsWorld}\")", affectsWorld:bool=@n6) -> (Result)
   )
   n7 = Get bCastShadows() -> (bCastShadows:bool)
   Set bCastShadows(execute=@n1.then_6, bCastShadows:bool=@n7) -> (Output_Get:bool)
   > "SetText (Text)"(
    self=Get Text_CastShadows() -> (Text_CastShadows:Text Object Reference)
    InText=Format Text!(Format="LOCTEXT(\"CastShadows: {castShadows}\")", castShadows:bool=@n7) -> (Result)
   )
   Set bVisible(execute=@n1.then_7, bVisible:bool=Get bVisible() -> (bVisible:bool)) -> (Output_Get:bool)
   n8 = Set Detail Mode(execute=@n1.then_8, Detail Mode:EDetailMode Enum=Get Detail Mode() -> (Detail Mode:EDetailMode Enum=DM_Low)) -> (Output_Get:EDetailMode Enum=DM_Low)
   > "SetText (Text)"(
    self=Get Text_DetailMode() -> (Text_DetailMode:Text Object Reference)
    InText=Format Text!(Format="LOCTEXT(\"Detail mode:  {detailMode}\")", detailMode:String=Enum to String(Enumerator:EDetailMode Enum=@n8)) -> (Result)
   )
  ]}
]
variables [
 ActorClassName {type=string, default=Default Name}
 ActorCount {type=integer}
 ActorsInLevel {type="array<Actor>"}
 bAffectsWorld {type=boolean}
 bBuildingInstance {type=boolean}
 bCastShadows {type=boolean}
 BuildingLevelInstances {type="array<Actor>"}
 bVisible {type=boolean}
 ClassRef {type=Class, default=null}
 ComponentClassName {type=string, default=Default Component Name}
 Detail Mode {type=EDetailMode, default=DM_Low}
 LightCountPerActor {type=integer}
 MaxDrawDistance {type=number}
 Mobility {type=EComponentMobility, default=Static}
]
widgets {
 info {rootWidgetClass=&Overlay, widgetCount=17}
 widgets [
  Overlay_1 {widgetClassPath=&Overlay, widget=Overlay_1}
  StackBox_0 {
   widgetClassPath=&StackBox
   parent=Overlay_1
   widget=StackBox_0
   properties {orientation=Orient_Vertical}
   slot {class=OverlaySlot, properties {horizontalAlignment=HAlign_Fill, verticalAlignment=VAlign_Fill}}
  }
  Border_71 {
   widgetClassPath=&Border
   parent=StackBox_0
   widget=Border_71
   properties {padding=<ltrb 1 1 1 1>, renderTransform {angle=0, scale=<1 0.5>, shear=<0 0>, translation=<0 0>}}
   slot {class=StackBoxSlot, properties {padding=<ltrb 10 10 10 0>}}
  }
  Text_ActorName {
   widgetClassPath=&TextBlock
   bIsVariable=true
   parent=StackBox_0
   widget=Text_ActorName
   properties {
    text="<<Actor Name>>"
    font {
     bForceMonospaced=false
     bMaterialIsStencil=false
     fontMaterial=null
     fontObject=&Roboto
     letterSpacing=0
     monospacedWidth=1
     size=10.5
     skewAmount=0
     typefaceFontName=Bold
     outlineSettings {
      bApplyOutlineToDropShadows=false
      bMiteredCorners=false
      bSeparateFillAlpha=false
      outlineMaterial=null
      outlineSize=0
      outlineColor=<rgba 0 0 0 1>
     }
    }
   }
   slot {class=StackBoxSlot, properties {padding=<ltrb 10 10 10 5>}}
  }
  Text_ActorCount {
   widgetClassPath=&TextBlock
   bIsVariable=true
   parent=StackBox_0
   widget=Text_ActorCount
   properties {
    text="<<ActorCount>>"
    font {
     bForceMonospaced=false
     bMaterialIsStencil=false
     fontMaterial=null
     fontObject=&Roboto
     letterSpacing=0
     monospacedWidth=1
     size=9
     skewAmount=0
     typefaceFontName=Bold
     outlineSettings {
      bApplyOutlineToDropShadows=false
      bMiteredCorners=false
      bSeparateFillAlpha=false
      outlineMaterial=null
      outlineSize=0
      outlineColor=<rgba 0 0 0 1>
     }
    }
   }
   slot {class=StackBoxSlot, properties {padding=<ltrb 10 3 10 3>}}
  }
  Text_ComponentName {
   widgetClassPath=&TextBlock
   bIsVariable=true
   parent=StackBox_0
   widget=Text_ComponentName
   properties {
    text="<<ComponentName>>"
    font {
     bForceMonospaced=false
     bMaterialIsStencil=false
     fontMaterial=null
     fontObject=&Roboto
     letterSpacing=0
     monospacedWidth=1
     size=9
     skewAmount=0
     typefaceFontName=Bold
     outlineSettings {
      bApplyOutlineToDropShadows=false
      bMiteredCorners=false
      bSeparateFillAlpha=false
      outlineMaterial=null
      outlineSize=0
      outlineColor=<rgba 0 0 0 1>
     }
    }
   }
   slot {class=StackBoxSlot, properties {padding=<ltrb 10 3 10 3>}}
  }
  Text_Mobility {
   widgetClassPath=&TextBlock
   bIsVariable=true
   parent=StackBox_0
   widget=Text_Mobility
   properties {
    text="<<Mobility>>"
    font {
     bForceMonospaced=false
     bMaterialIsStencil=false
     fontMaterial=null
     fontObject=&Roboto
     letterSpacing=0
     monospacedWidth=1
     size=9
     skewAmount=0
     typefaceFontName=Bold
     outlineSettings {
      bApplyOutlineToDropShadows=false
      bMiteredCorners=false
      bSeparateFillAlpha=false
      outlineMaterial=null
      outlineSize=0
      outlineColor=<rgba 0 0 0 1>
     }
    }
   }
   slot {class=StackBoxSlot, properties {padding=<ltrb 10 3 10 3>}}
  }
  Text_MaxDrawDistance {
   widgetClassPath=&TextBlock
   bIsVariable=true
   parent=StackBox_0
   widget=Text_MaxDrawDistance
   properties {
    text="<<MaxDrawDistance>>"
    font {
     bForceMonospaced=false
     bMaterialIsStencil=false
     fontMaterial=null
     fontObject=&Roboto
     letterSpacing=0
     monospacedWidth=1
     size=9
     skewAmount=0
     typefaceFontName=Bold
     outlineSettings {
      bApplyOutlineToDropShadows=false
      bMiteredCorners=false
      bSeparateFillAlpha=false
      outlineMaterial=null
      outlineSize=0
      outlineColor=<rgba 0 0 0 1>
     }
    }
   }
   slot {class=StackBoxSlot, properties {padding=<ltrb 10 3 10 3>}}
  }
  Text_AffectsWorld {
   widgetClassPath=&TextBlock
   bIsVariable=true
   parent=StackBox_0
   widget=Text_AffectsWorld
   properties {
    text="<<AffectsWorld>>"
    font {
     bForceMonospaced=false
     bMaterialIsStencil=false
     fontMaterial=null
     fontObject=&Roboto
     letterSpacing=0
     monospacedWidth=1
     size=9
     skewAmount=0
     typefaceFontName=Bold
     outlineSettings {
      bApplyOutlineToDropShadows=false
      bMiteredCorners=false
      bSeparateFillAlpha=false
      outlineMaterial=null
      outlineSize=0
      outlineColor=<rgba 0 0 0 1>
     }
    }
   }
   slot {class=StackBoxSlot, properties {padding=<ltrb 10 3 10 3>}}
  }
  Text_CastShadows {
   widgetClassPath=&TextBlock
   bIsVariable=true
   parent=StackBox_0
   widget=Text_CastShadows
   properties {
    text="<<CastShadows>>"
    font {
     bForceMonospaced=false
     bMaterialIsStencil=false
     fontMaterial=null
     fontObject=&Roboto
     letterSpacing=0
     monospacedWidth=1
     size=9
     skewAmount=0
     typefaceFontName=Bold
     outlineSettings {
      bApplyOutlineToDropShadows=false
      bMiteredCorners=false
      bSeparateFillAlpha=false
      outlineMaterial=null
      outlineSize=0
      outlineColor=<rgba 0 0 0 1>
     }
    }
   }
   slot {class=StackBoxSlot, properties {padding=<ltrb 10 3 10 3>}}
  }
  Text_DetailMode {
   widgetClassPath=&TextBlock
   bIsVariable=true
   parent=StackBox_0
   widget=Text_DetailMode
   properties {
    text="<<DetailMode>>"
    font {
     bForceMonospaced=false
     bMaterialIsStencil=false
     fontMaterial=null
     fontObject=&Roboto
     letterSpacing=0
     monospacedWidth=1
     size=9
     skewAmount=0
     typefaceFontName=Bold
     outlineSettings {
      bApplyOutlineToDropShadows=false
      bMiteredCorners=false
      bSeparateFillAlpha=false
      outlineMaterial=null
      outlineSize=0
      outlineColor=<rgba 0 0 0 1>
     }
    }
   }
   slot {class=StackBoxSlot, properties {padding=<ltrb 10 3 10 3>}}
  }
  Text_BuildingLevelInstanceWarning {
   widgetClassPath=&TextBlock
   bIsVariable=true
   parent=StackBox_0
   widget=Text_BuildingLevelInstanceWarning
   properties {
    text=Building Level Instance
    visibility=Collapsed
    colorAndOpacity {colorUseRule=UseColor_Specified, specifiedColor=<rgba 0.0586 0.5 0.3581 1>}
    font {
     bForceMonospaced=false
     bMaterialIsStencil=false
     fontMaterial=null
     fontObject=&Roboto
     letterSpacing=0
     monospacedWidth=1
     size=9
     skewAmount=0
     typefaceFontName=Bold
     outlineSettings {
      bApplyOutlineToDropShadows=false
      bMiteredCorners=false
      bSeparateFillAlpha=false
      outlineMaterial=null
      outlineSize=0
      outlineColor=<rgba 0 0 0 1>
     }
    }
   }
   slot {class=StackBoxSlot, properties {padding=<ltrb 10 3 10 3>}}
  }
  HorizontalBox_0 {
   widgetClassPath=&HorizontalBox
   parent=Overlay_1
   widget=HorizontalBox_0
   slot {class=OverlaySlot, properties {horizontalAlignment=HAlign_Right, verticalAlignment=VAlign_Bottom, padding=<ltrb 10 10 10 10>}}
  }
  Button_SelectInstances {
   widgetClassPath=&EditorUtilityButton
   parent=HorizontalBox_0
   widget=Button_SelectInstances
   properties {toolTipText="Selects containing building instance  actor", visibility=Collapsed, backgroundColor=<rgba 0.1141 0.974 0.6975 1>}
   slot {class=HorizontalBoxSlot, properties {padding=<ltrb 10 0 10 0>}}
  }
  Image_107 {
   widgetClassPath=&Image
   parent=Button_SelectInstances
   widget=Image_107
   properties {brush {
     bIsDynamicallyLoaded=false
     drawAs=Image
     imageType=FullColor
     mirroring=NoMirror
     resourceName=None
     resourceObject=/Game/UI/Friends_UI/Social/Search
     tiling=NoTile
     imageSize=<20 20>
     margin=<ltrb 0 0 0 0>
     outlineSettings {
      bUseBrushTransparency=false
      roundingType=HalfHeightRadius
      width=0
      color {colorUseRule=UseColor_Specified, specifiedColor=<rgba 0 0 0 0>}
      cornerRadii=<0 0 0 0>
     }
     tintColor {colorUseRule=UseColor_Specified, specifiedColor=<rgba 1 1 1 1>}
     uVRegion {bIsValid=false, max=<0 0>, min=<0 0>}
    }}
   slot {class=ButtonSlot}
  }
  Button_SelectLights {
   widgetClassPath=&EditorUtilityButton
   parent=HorizontalBox_0
   widget=Button_SelectLights
   properties {toolTipText=Select Actors}
   slot {class=HorizontalBoxSlot}
  }
  Image {
   widgetClassPath=&Image
   parent=Button_SelectLights
   widget=Image
   properties {brush {
     bIsDynamicallyLoaded=false
     drawAs=Image
     imageType=FullColor
     mirroring=NoMirror
     resourceName=None
     resourceObject=&Search
     tiling=NoTile
     imageSize=<20 20>
     margin=<ltrb 0 0 0 0>
     outlineSettings {
      bUseBrushTransparency=false
      roundingType=HalfHeightRadius
      width=0
      color {colorUseRule=UseColor_Specified, specifiedColor=<rgba 0 0 0 0>}
      cornerRadii=<0 0 0 0>
     }
     tintColor {colorUseRule=UseColor_Specified, specifiedColor=<rgba 1 1 1 1>}
     uVRegion {bIsValid=false, max=<0 0>, min=<0 0>}
    }}
   slot {class=ButtonSlot}
  }
 ]
}
