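# Editor utility widget for track performance review that lists and counts lights in the current level.
# On construct or refresh, it scans all or selected actors, including those inside Building Level Instances, applying visibility, affects-world and shadow-casting filters from a self-bound Details View.
# Matches are grouped by actor class into entry widgets showing light properties.
# Export builds a CSV from the entries and writes it to the project's Saved directory through an inline Python command.
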EditorUtilityWidgetBlueprint /DelMarTools/EditorUtilities/AssetReviewTools/TrackPerformanceReview/Subwidgets/LevelLightTracker/LevelLightTrackerTool
graphs [
 Check for Lights {nodes [
   Check for Lights!() -> (then, ActorToCheck:Actor Object Reference, BuildingInstance:Actor Object Reference) {is=title K2Node_FunctionEntry}
   > n1 = Sequence!(execute) -> (then_0, then_1, then_2:Exec, then_3:Exec)
   n2 = GetComponentByClass!(
    self:Actor Object Reference=Get ActorToCheck() -> (ActorToCheck:Actor Object Reference)
    ComponentClass:Actor Component Class Reference=/Script/Engine.LightComponent
   ) -> (ReturnValue:Light Component Object Reference)
   > n3 = Branch(Condition=IsValid(Object=@n2))
   > Set LightComponent(LightComponent:Light Component Object Reference=@n2) -> (Output_Get:Light Component Object Reference)
   Return Node(execute=@n3.else)
   n4 = Get Class(Object=Get ActorToCheck())
   n5 = Find(
    TargetMap:Map of Actor Class References to Level Light Tracker Tool Entry Object References=Get Entries() -> (Entries:Map of Actor Class References to Level Light Tracker Tool Entry Object References)
    Key:"Actor Class Reference (by ref)"=@n4
   ) -> (Value:Level Light Tracker Tool Entry Object Reference)
   n6 = Branch(execute=@n1.then_3, Condition=@n5.ReturnValue)
   > n7 = Set EntryWidget(EntryWidget:Level Light Tracker Tool Entry Object Reference=@n5) -> (Output_Get:Level Light Tracker Tool Entry Object Reference)
   AddActor!(
    execute=[@n14.then, @n7.then]
    self:Level Light Tracker Tool Entry Object Reference=Get EntryWidget() -> (EntryWidget:Level Light Tracker Tool Entry Object Reference)
    Actor:Actor Object Reference=Get ActorToCheck()
   ) -> (then)
   n8 = Get BuildingInstance() -> (BuildingInstance:Actor Object Reference)
   > Is Valid(InputObject=@n8)
   > AddBuildingLevelInstance!(execute, self:Level Light Tracker Tool Entry Object Reference=Get EntryWidget(), BuildingLevelInstance:Actor Object Reference=@n8) -> (then)
   n9 = Create Widget!(
    execute:Exec=@n6.else
    WorldContextObject:Object Reference
    Class:User Widget Class Reference=&/LevelLightTrackerTool_Entry.LevelLightTrackerTool_Entry_C
    OwningPlayer:Player Controller Object Reference
    bBuildingInstance:bool=IsValid(Object=Get BuildingInstance())
    ClassRef:Actor Class Reference=@n4
   ) -> (then:Exec, ReturnValue:Level Light Tracker Tool Entry Object Reference)
   > Set EntryWidget(EntryWidget=@n9)
   > n10 = GetEditorProperty!(
    execute:Exec
    Object:Object Reference=Get LightComponent() -> (LightComponent:Light Component Object Reference)
    PropertyName:Name=MaxDrawDistance
   ) -> (then:Exec, PropertyValue:double, ReturnValue:bool)
   n11 = Get ActorToCheck()
   n12 = Get LightComponent()
   n13 = Get ActorToCheck()
   > UpdateInfo!(
    execute
    self:Level Light Tracker Tool Entry Object Reference=Get EntryWidget()
    ActorClassName:String=Get Class Display Name(Class=Get Class(Object=@n13))
    ComponentClassName:String=Get Class Display Name(Class=Get Class(Object=@n12) -> (ReturnValue:Light Component Class Reference))
    LightCountPerActor:int=Length(TargetArray:Array of Light Component Object References=Get Components By Class!(self:Actor Object Reference=@n11, ComponentClass:Actor Component Class Reference=&LightComponent) -> (ReturnValue:Array of Light Component Object References))
    bVisible:bool=Get bVisible!(self:Scene Component Object Reference=@n12) -> (bVisible:bool)
    bAffectsWorld:bool=Get bAffectsWorld!(self:Light Component Base Object Reference=@n12) -> (bAffectsWorld:bool)
    bCastShadows:bool=Get CastShadows!(self:Light Component Base Object Reference=@n12) -> (CastShadows:bool)
    Mobility:EComponentMobility Enum=Get Mobility!(self:Scene Component Object Reference=@n12) -> (Mobility:EComponentMobility Enum=Static)
    Detail Mode:EDetailMode Enum=Get DetailMode!(self:Scene Component Object Reference=@n12) -> (DetailMode:EDetailMode Enum=DM_Low)
    MaxDrawDistance:double=@n10
   ) -> (then)
   > Add!(
    execute
    TargetMap:Map of Actor Class References to Level Light Tracker Tool Entry Object References=Get Entries()
    Key:"Actor Class Reference (by ref)"=Get Class(Object=Get ActorToCheck()) -> (ReturnValue:Actor Class Reference)
    Value:"Level Light Tracker Tool Entry Object Reference (by ref)"=Get EntryWidget()
   ) -> (then) {is=title K2Node_CallFunction}
   > n14 = AddChild(self=Get ScrollBox_Entries() -> (ScrollBox_Entries:Editor Utility Scroll Box Object Reference), Content=Get EntryWidget())
   n15 = NOT Boolean(A=IsVisible(self:Scene Component Object Reference=Get LightComponent()))
   n16 = Get bVisibleOnly() -> (bVisibleOnly:bool)
   n17 = NOT Boolean(A=Get bAffectsWorld(self=Get LightComponent()))
   n18 = Get bAffectWorldOnly() -> (bAffectWorldOnly:bool)
   n19 = NOT Boolean(A=Get CastShadows(self=Get LightComponent()))
   n20 = Get bCastShadowOnly() -> (bCastShadowOnly:bool)
   Branch(
    execute=@n1.then_1
    Condition=OR Boolean!(A=AND Boolean(A=@n16, B=@n15), B=AND Boolean(A=@n18, B=@n17), C:bool=AND Boolean(A=@n20, B=@n19)) -> (ReturnValue)
   )
   > Return Node()
   n21 = Get Components By Class(self=Get ActorToCheck(), ComponentClass=&LightComponent)
   Set LightCount(execute=@n1.then_2, LightCount:int=int + int(A=Get LightCount() -> (LightCount:int=0), B=Length(TargetArray=@n21))) -> (Output_Get:int=0)
  ]}
 EventGraph {nodes [
   n1 = "On Clicked (Button_RefreshLights)"() {is=title K2Node_ComponentBoundEvent}
   n2 = Sequence!(execute=[@n1.then, @n15.then]) -> (then_0, then_1)
   > Set LightCount(LightCount=0) -> (Output_Get=0)
   > n3 = Values!(execute, TargetMap:Map of Actor Class References to Level Light Tracker Tool Entry Object References=Get Entries()) -> (then, Values:Array of Level Light Tracker Tool Entry Object References)
   > RemoveFromParent(self=@n3)
   > Clear!(execute, TargetMap:Map of Actor Class References to Level Light Tracker Tool Entry Object References=Get Entries()) -> (then) {is=title K2Node_CallFunction}
   n4 = Branch(execute=@n2.then_1, Condition=Get bSelectedActorsOnly() -> (bSelectedActorsOnly:bool))
   > n5 = GetSelectedLevelActors(self=Get EditorActorSubsystem())
   > n6 = Set ActorsToSearch(ActorsToSearch:Array of Actor Object References=@n5) -> (Output_Get:Array of Actor Object References)
   n7 = GetAllLevelActors(execute=@n4.else, self=Get EditorActorSubsystem())
   > n8 = Set ActorsToSearch(ActorsToSearch=@n7)
   n9 = For Each Loop(Exec=[@n6.then, @n8.then], Array:Array of Actor Object References=@n8) -> (Array Element:Actor Object Reference)
   > n10 = Sequence()
   > n11 = Cast To BuildingLevelInstance!(execute, Object=@n9) -> (then, CastFailed, AsBuilding Level Instance:Building Level Instance Object Reference) {is=title K2Node_DynamicCast}
   > n12 = GetActorsInLevel!(execute, self:Building Level Instance Object Reference=@n11) -> (then, ReturnValue:Array of Actor Object References)
   > n13 = For Each Loop(Array=@n12)
   > Check for Lights!(execute, self, ActorToCheck:Actor Object Reference=@n13, BuildingInstance:Actor Object Reference=@n11) -> (then) {is=title K2Node_CallFunction}
   Check for Lights(execute=@n10.then_1, ActorToCheck=@n9)
   n14 = "To String (Integer)"(InInt=Get LightCount() -> (LightCount=0))
   "SetText (Text)"(
    execute=@n9.Completed
    self=Get Text_AmountOfLights() -> (Text_AmountOfLights:Text Object Reference)
    InText="To Text (String)"(InString=Append!(A=@n14, B=" lights found.", C:String) -> (ReturnValue))
   )
   "On Clicked (Button_ExportCSV)"() {is=title K2Node_ComponentBoundEvent}
   > ExportCSV()
   Event PreConstruct()
   > SetObject(self=Get DetailsView_Parameters() -> (DetailsView_Parameters:Details View Object Reference), NewObject=Self-Reference())
   n15 = Event Construct()
  ]}
 ExportCSV {nodes [
   ExportCSV() {is=title K2Node_FunctionEntry}
   > n1 = Sequence()
   > n2 = GetCurrentLevelName!(execute, WorldContextObject:Object Reference=Self-Reference(), bRemovePrefixString:bool=true) -> (then, ReturnValue:String)
   > n3 = GetMcpRelativeUtcNow!(execute, FortPC:Fort Player Controller Object Reference) -> (then, ReturnValue:Date Time Structure)
   n4 = BreakDateTime!(InDateTime:Date Time Structure=@n3) -> (Year:int=0, Month:int=0, Day:int=0, Hour:int=0, Minute:int=0, Second:int=0, Millisecond:int=0)
   > Set FileName(FileName:Text=Format Text!(
     Format="LOCTEXT(\"LightInfo_{TrackName}_{Year}-{Month}-{Day}_{Hour}-{Minute}-{Time}\")"
     TrackName:String=@n2
     Year:String="To String (Integer)"(InInt=@n4)
     Month:int=@n4.Month
     Day:int=@n4.Day
     Hour:int=@n4.Hour
     Minute:int=@n4.Minute
     Time:int=@n4.Second
    ) -> (Result)) -> (Output_Get:Text)
   Set ExportString(
    execute=@n1.then_1
    ExportString:String="ActorClassName, ActorInstanceCount, ComponentType, ComponentCount, Mobility, Visible, AffectsWorld, CastShadows, MaxDrawDistance,  DetailMode"
   ) -> (Output_Get:String)
   > n5 = Values(TargetMap=Get Entries())
   > n6 = For Each Loop(Array:Array of Level Light Tracker Tool Entry Object References=@n5) -> (Array Element:Level Light Tracker Tool Entry Object Reference)
   n7 = Get ExportString() -> (ExportString:String)
   n8 = Get FileName() -> (FileName:Text)
   "Execute Python Command (Advanced)"!(
    execute=@n6.Completed
    PythonCommand:String="To String (Text)"(InText=Format Text!(
      Format="INVTEXT(\"f=open(unreal.Paths.project_saved_dir()+\\\"{File Name}.csv\\\", \\\"w\\\")\\r\\nf.write(\\\"\\\"\\\"{content}\\\"\\\"\\\")\\r\\nf.close()\")"
      File Name:Text=@n8
      content:String=@n7
     ) -> (Result))
    ExecutionMode:EPythonCommandExecutionMode Enum=ExecuteFile
    FileExecutionScope:EPythonFileExecutionScope Enum=Private
   ) -> (then, CommandResult:String, LogOutput:Array of Python Log Output Entry Structures, ReturnValue:bool)
   > PrintString(InString=Get ExportString())
   > PrintText(InText=Format Text!(
     Format="LOCTEXT(\"Light Info CSV was exported to {SavedDir}{FileName}.csv\")"
     SavedDir:String=GetProjectSavedDirectory!() -> (ReturnValue:String)
     FileName:Text=Get FileName()
    ) -> (Result))
   n9 = Format Text!(
    Format="LOCTEXT(\"\\r\\n{ActorName}, {ActorCount}, {Component}, {ComponentCount}, {Mobility}, {Visible}, {AffectsWorld}, {CastShadows}, \\\"{MaxDrawDistance}\\\", {DetailMode}\")"
    ActorName:String=Get ActorClassName!(self:Level Light Tracker Tool Entry Object Reference=@n6) -> (ActorClassName:String)
    ActorCount:int=Get ActorCount!(self:Level Light Tracker Tool Entry Object Reference=@n6) -> (ActorCount:int=0)
    Component:String=Get ComponentClassName!(self:Level Light Tracker Tool Entry Object Reference=@n6) -> (ComponentClassName:String)
    ComponentCount:int=Get LightCountPerActor!(self:Level Light Tracker Tool Entry Object Reference=@n6) -> (LightCountPerActor:int=0)
    Mobility:String=Enum to String!(Enumerator:EComponentMobility Enum=Get Mobility(self:Level Light Tracker Tool Entry Object Reference=@n6) -> (Mobility=Static)) -> (ReturnValue:String)
    Visible:bool=Get bVisible(self:Level Light Tracker Tool Entry Object Reference=@n6)
    AffectsWorld:bool=Get bAffectsWorld(self:Level Light Tracker Tool Entry Object Reference=@n6)
    CastShadows:bool=Get bCastShadows!(self:Level Light Tracker Tool Entry Object Reference=@n6) -> (bCastShadows:bool)
    MaxDrawDistance:double=Get MaxDrawDistance!(self:Level Light Tracker Tool Entry Object Reference=@n6) -> (MaxDrawDistance:double)
    DetailMode:String=Enum to String(Enumerator:EDetailMode Enum=Get Detail Mode!(self:Level Light Tracker Tool Entry Object Reference=@n6) -> (Detail Mode:EDetailMode Enum=DM_Low))
   ) -> (Result)
   Set ExportString(execute=@n6.LoopBody, ExportString=Append!(A=Get ExportString(), B="To String (Text)"(InText=@n9)) -> (ReturnValue))
  ]}
]
variables [
 ActorsToSearch {type="array<Actor>"}
 bAffectWorldOnly {type=boolean}
 bCastShadowOnly {type=boolean}
 bSelectedActorsOnly {type=boolean}
 bVisibleOnly {type=boolean}
 Detail Mode {type=EDetailMode, default=DM_Low}
 Entries {default {}}
 LightCount {type=integer}
]
widgets {
 info {rootWidgetClass=&StackBox, widgetCount=15}
 widgets [
  StackBox_0 {widgetClassPath=&StackBox, widget=StackBox_0, properties {orientation=Orient_Vertical}}
  VerticalBox_29 {
   widgetClassPath=&VerticalBox
   parent=StackBox_0
   widget=VerticalBox_29
   slot {class=StackBoxSlot, properties {padding=<ltrb 10 10 10 10>, size {sizeRule=Fill, value=1}}}
  }
  DetailsView_Parameters {
   widgetClassPath=&DetailsView
   parent=VerticalBox_29
   widget=DetailsView_Parameters
   properties {categoriesToShow [SearchParameters, SearchParameters | Mobility]}
   slot {class=VerticalBoxSlot}
  }
  Spacer {
   widgetClassPath=&Spacer
   parent=VerticalBox_29
   widget=Spacer
   slot {class=VerticalBoxSlot, properties {padding=<ltrb 10 10 10 10>}}
  }
  Button_RefreshLights {widgetClassPath=&EditorUtilityButton, parent=VerticalBox_29, widget=Button_RefreshLights, slot {class=VerticalBoxSlot}}
  Text_Button {
   widgetClassPath=&TextBlock
   parent=Button_RefreshLights
   widget=Text_Button
   properties {
    text=Refresh Light List
    font {
     bForceMonospaced=false
     bMaterialIsStencil=false
     fontMaterial=null
     fontObject=&Roboto
     letterSpacing=0
     monospacedWidth=1
     size=12
     skewAmount=0
     typefaceFontName=Bold
     outlineSettings {
      bApplyOutlineToDropShadows=false
      bMiteredCorners=false
      bSeparateFillAlpha=false
      outlineMaterial=null
      outlineSize=0
      outlineColor=<rgba 0 0 0 1>
     }
    }
   }
   slot {class=ButtonSlot, properties {padding=<ltrb 4 10 4 10>}}
  }
  Spacer_119 {
   widgetClassPath=&Spacer
   parent=VerticalBox_29
   widget=Spacer_119
   slot {class=VerticalBoxSlot, properties {padding=<ltrb 10 10 10 10>}}
  }
  SizeBox_0 {
   widgetClassPath=&SizeBox
   parent=VerticalBox_29
   widget=SizeBox_0
   properties {heightOverride=593.0844}
   slot {class=VerticalBoxSlot, properties {size {sizeRule=Fill, value=1}}}
  }
  Border_0 {
   widgetClassPath=&Border
   parent=SizeBox_0
   widget=Border_0
   properties {brushColor=<rgba 0.01 0.01 0.01 1>, padding=<ltrb 10 2 10 2>}
   slot {class=SizeBoxSlot}
  }
  ScrollBox_Entries {
   widgetClassPath=&EditorUtilityScrollBox
   bIsVariable=true
   parent=Border_0
   widget=ScrollBox_Entries
   properties {alwaysShowScrollbar=true, alwaysShowScrollbarTrack=true}
   slot {class=BorderSlot, properties {padding=<ltrb 10 2 10 2>}}
  }
  Spacer_1 {
   widgetClassPath=&Spacer
   parent=VerticalBox_29
   widget=Spacer_1
   slot {class=VerticalBoxSlot, properties {padding=<ltrb 10 10 10 10>}}
  }
  Overlay_0 {widgetClassPath=&Overlay, parent=VerticalBox_29, widget=Overlay_0, slot {class=VerticalBoxSlot}}
  Text_AmountOfLights {
   widgetClassPath=&TextBlock
   bIsVariable=true
   parent=Overlay_0
   widget=Text_AmountOfLights
   properties {
    text=0 lights found
    font {
     bForceMonospaced=false
     bMaterialIsStencil=false
     fontMaterial=null
     fontObject=&Roboto
     letterSpacing=0
     monospacedWidth=1
     size=9
     skewAmount=0
     typefaceFontName=Bold
     outlineSettings {
      bApplyOutlineToDropShadows=false
      bMiteredCorners=false
      bSeparateFillAlpha=false
      outlineMaterial=null
      outlineSize=0
      outlineColor=<rgba 0 0 0 1>
     }
    }
   }
   slot {class=OverlaySlot}
  }
  Button_ExportCSV {
   widgetClassPath=&EditorUtilityButton
   parent=Overlay_0
   widget=Button_ExportCSV
   slot {class=OverlaySlot, properties {horizontalAlignment=HAlign_Right, verticalAlignment=VAlign_Fill, padding=<ltrb 0 0 0 5>}}
  }
  Text_ExportCSV {
   widgetClassPath=&TextBlock
   parent=Button_ExportCSV
   widget=Text_ExportCSV
   properties {
    text=Export CSV
    font {
     bForceMonospaced=false
     bMaterialIsStencil=false
     fontMaterial=null
     fontObject=&Roboto
     letterSpacing=0
     monospacedWidth=1
     size=9
     skewAmount=0
     typefaceFontName=Bold
     outlineSettings {
      bApplyOutlineToDropShadows=false
      bMiteredCorners=false
      bSeparateFillAlpha=false
      outlineMaterial=null
      outlineSize=0
      outlineColor=<rgba 0 0 0 1>
     }
    }
   }
   slot {class=ButtonSlot}
  }
 ]
}
